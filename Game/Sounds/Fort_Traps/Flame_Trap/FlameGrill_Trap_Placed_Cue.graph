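# Sound cue played when a flame grill trap is placed.
# A single wave player feeds a modulator at the root for slight variation on each play.
# It uses shared trap attenuation and the trap placement sound class.
# A concurrency override caps it at two simultaneous instances, stopping the oldest.
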SoundCue /Game/Sounds/Fort_Traps/Flame_Trap/FlameGrill_Trap_Placed_Cue
root=Modulator_0
nodes [
 {id=Modulator_0, children [WavePlayer_0]}
 {id=WavePlayer_0, properties {soundWaveAssetPtr=&/FlameGrill_Trap_Place_01}}
]
properties {
 attenuationSettings=&Fort_Traps_Att
 bOverrideConcurrency=true
 maxDistance=6000
 soundClassObject=/Game/Sounds/FortSoundClassesAndModes/Fort_Traps_Placement
 volumeMultiplier=1
 concurrencyOverrides {maxCount=2, resolutionRule=StopOldest}
}
