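# Sound cue for the pre-fire warning sound of the flame grill trap, judging by its name.
# A single wave player plays the FlameGrill_Trap_PreFire wave with a slight volume boost and no randomization.
# It uses the shared trap attenuation settings and the trap placement sound class.
# A concurrency override prevents new instances from starting once the limit is reached.
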SoundCue /Game/Sounds/Fort_Traps/Flame_Trap/FlameGrill_Trap_PreFire_Cue
root=WavePlayer_2
nodes [{id=WavePlayer_2, properties {soundWaveAssetPtr=&/FlameGrill_Trap_PreFire}}]
properties {
 attenuationSettings=&Fort_Traps_Att
 bOverrideConcurrency=true
 maxDistance=6000
 soundClassObject=/Game/Sounds/FortSoundClassesAndModes/Fort_Traps_Placement
 volumeMultiplier=1.2
 concurrencyOverrides {resolutionRule=PreventNew}
}
