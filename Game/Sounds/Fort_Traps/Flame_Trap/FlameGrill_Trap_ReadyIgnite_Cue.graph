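# Sound cue for the Fort flame grill trap's ready-ignite sound, the pilot ignition.
# A single wave player feeds the pilot ignite sound through a modulator at the root.
# It uses the shared trap attenuation and sound class, and overrides concurrency so new instances are blocked rather than cutting off playing ones when the limit is reached.
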SoundCue /Game/Sounds/Fort_Traps/Flame_Trap/FlameGrill_Trap_ReadyIgnite_Cue
root=Modulator_0
nodes [
 {id=Modulator_0, children [WavePlayer_1]}
 {id=WavePlayer_1, properties {soundWaveAssetPtr=&/FlameTrap_PilotIgnite_01}}
]
properties {
 attenuationSettings=&Fort_Traps_Att
 bOverrideConcurrency=true
 maxDistance=6000
 soundClassObject=&Fort_Traps
 volumeMultiplier=1
 concurrencyOverrides {resolutionRule=PreventNew}
}
