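# Sound cue for the Flame Grill trap's ready sound in the Fort traps audio set.
# A single wave player feeds the root directly, playing one ready sample with no randomization or modulation.
# It uses the shared traps attenuation and sound class, and overrides concurrency so new instances are blocked rather than stopping playing ones.
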SoundCue /Game/Sounds/Fort_Traps/Flame_Trap/FlameGrill_Trap_Ready_Cue
root=WavePlayer_0
nodes [{id=WavePlayer_0, properties {soundWaveAssetPtr=&/FlameGrill_Trap_Ready_01}}]
properties {
 attenuationSettings=&Fort_Traps_Att
 bOverrideConcurrency=true
 maxDistance=6000
 soundClassObject=&Fort_Traps
 volumeMultiplier=1
 concurrencyOverrides {resolutionRule=PreventNew}
}
