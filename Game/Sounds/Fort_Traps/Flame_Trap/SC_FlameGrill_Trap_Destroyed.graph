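# Sound cue for the destruction of the Flame Grill trap.
# A random node with equal weights picks one of two destroy wave variants.
# Playback runs at reduced volume through the Fort_Traps sound class and shared traps attenuation.
# A concurrency override caps simultaneous instances at a small count and blocks new plays beyond it rather than stopping older ones.
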SoundCue /Game/Sounds/Fort_Traps/Flame_Trap/SC_FlameGrill_Trap_Destroyed
root=Random_0
nodes [
 {id=Random_0, children [WavePlayer_0,WavePlayer_1], properties {weights [1,1]}}
 {id=WavePlayer_0, properties {soundWaveAssetPtr=&/FlameGrill_Trap_Destroy_01}}
 {id=WavePlayer_1, properties {soundWaveAssetPtr=&/FlameGrill_Trap_Destroy_02}}
]
properties {
 attenuationSettings=&Fort_Traps_Att
 bOverrideConcurrency=true
 maxDistance=6000
 soundClassObject=&Fort_Traps
 volumeMultiplier=0.5
 concurrencyOverrides {maxCount=3, resolutionRule=PreventNew}
}
