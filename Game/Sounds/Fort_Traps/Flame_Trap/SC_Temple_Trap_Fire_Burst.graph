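# Sound cue for the fire burst of the Temple flame trap, played as a one-shot effect in the Fort traps audio set.
# A Random node picks with equal weight between two FlameGrill trap fire recordings.
# It plays at reduced volume through the Fort trap attenuation and the trap placement sound class.
# A concurrency override caps simultaneous instances, stopping the farthest before blocking new ones.
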SoundCue /Game/Sounds/Fort_Traps/Flame_Trap/SC_Temple_Trap_Fire_Burst
root=Random_0
nodes [
 {id=Random_0, children [WavePlayer_0,WavePlayer_1], properties {weights [1,1]}}
 {id=WavePlayer_0, properties {soundWaveAssetPtr=&/FlameGrill_Trap_Fire_02}}
 {id=WavePlayer_1, properties {soundWaveAssetPtr=&/FlameGrill_Trap_Fire_01}}
]
properties {
 attenuationSettings=&Fort_Traps_Att
 bOverrideConcurrency=true
 maxDistance=6000
 soundClassObject=/Game/Sounds/FortSoundClassesAndModes/Fort_Traps_Placement
 volumeMultiplier=0.3
 concurrencyOverrides {maxCount=3, resolutionRule=StopFarthestThenPreventNew}
}
