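# Sound cue for the looping flame sound of the Temple fire trap in the Fort traps set.
# A single looping wave player plays the thruster flame loop at reduced volume through the shared trap attenuation and the trap placement sound class.
# A concurrency override caps simultaneous instances at a few, stopping the farthest before preventing new ones.
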SoundCue /Game/Sounds/Fort_Traps/Flame_Trap/SC_Temple_Trap_Fire_Loop
root=WavePlayer_2
nodes [{id=WavePlayer_2, properties {bLooping=true, soundWaveAssetPtr=&/TempleTrap_Thruster_Flame_Loop_01}}]
properties {
 attenuationSettings=&Fort_Traps_Att
 bOverrideConcurrency=true
 maxDistance=6000
 soundClassObject=/Game/Sounds/FortSoundClassesAndModes/Fort_Traps_Placement
 volumeMultiplier=0.3
 concurrencyOverrides {maxCount=3, resolutionRule=StopFarthestThenPreventNew}
}
